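; Figment cube event helper that removes desert cacti from a preset list as the event's cactus hit index progresses.
; On authority it binds to event component delegates for live destruction, or catches up immediately if hits already happened; cube setup completion clears up to 20.
; Cleanup spawns lightning decals, then auto-destroys or silently kills each cactus.
; A RepNotify places scorch decals on remote clients.
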(event Custom|OnCubeSetupComplete
  (params (CurrentIndex :type Integer))
  ;[1 1 1] On Cube spawn, destroy cactus again
  (CallFunction|CleanUpCacti 20))

(event EventBeginPlay
  (Networking|SwitchHasAuthority
    (:Authority
      (Utilities|FlowControl|Sequence
        (:then_0
          ;[1 1 1] On Cube spawn, destroy cactus again
          (bind _outactors (Actor|GetAllActorsOfClass "/Figment_S05_Cube/Athena/Prototype/Blueprints/Cube/CUBE_Figment.CUBE_Figment_C"))
          (Default|BindEventtoCubeSetupComplete (Utilities|Array|Get(acopy) _outactors) (delegate Custom|OnCubeSetupComplete)))
        (:then_1
          (bind _outactors_2 (Actor|GetAllActorsOfClass "/Figment_S05_Cube/Athena/Events/BP_Athena_Event_Components_Old.BP_Athena_Event_Components_Old_C"))
          (if (== (Variables|/Figment_S05_Cube/Athena/Events/BP_Athena_Event_Components_Old.BP_Athena_Event_Components_Old_C|GetCactusHitIndex (Utilities|Array|Get(acopy) _outactors_2)) 0)
            (Default|BindEventtoCactusCalculationsComplete (Utilities|Array|Get(acopy) _outactors_2) (delegate Custom|CleanUpCacti))
            (Default|BindEventtoDestroyCactus (Utilities|Array|Get(acopy) _outactors_2) (delegate Custom|DestroyCactus))
            (else
              (CallFunction|CleanUpCacti (Math|Conversions|ToInteger(Byte) (Variables|/Figment_S05_Cube/Athena/Events/BP_Athena_Event_Components_Old.BP_Athena_Event_Components_Old_C|GetCactusHitIndex (Utilities|Array|Get(acopy) _outactors_2)))))))))))

(event Custom|DestroyCactus
  (params (Index :type Integer))
  (bind _returnvalue (- Index 1))
  (bind _cactus (Variables|Default|GetCactus))
  (bind _output (Utilities|Array|Get(acopy) _cactus _returnvalue))
  (Utilities|IsValid _output
    (:Is%20Valid
      (Actor|DestroyActor _output))))

(event Custom|CleanUpCacti
  (params (CactusHitIndex :type Integer))
  (for (_array_element _array_index) (Variables|Default|GetCactus)
    (if (>= (- CactusHitIndex 1) _array_index)
      (Utilities|FlowControl|Sequence
        (:then_0
          (Utilities|Casting|CastToParent_Tree _array_element)
          (bind (_returnvalue_location _returnvalue_rotation) (Math|Transform|BreakTransform (Transformation|GetActorTransform _array_element)))
          (Game|SpawnActorfromClass "/FNGameplayCues/Athena/Prototype/Blueprints/Cube/FX/BP_Decal_Lightning.BP_Decal_Lightning_C" (Math|Transform|MakeTransform _returnvalue_location _returnvalue_rotation) "Undefined" "SelectDefaultAtRuntime" 0 0))
        (:then_1
          (bind _asprop_desert_cactus_02 (Utilities|Casting|CastToProp_Desert_Cactus_02 _array_element)
            (:then
              (Variables|/Game/Building/ActorBlueprints/Prop/Prop_Desert_Cactus_02.Prop_Desert_Cactus_02_C|SetAutoDestroy true _asprop_desert_cactus_02))
            (:CastFailed
              (Gameplay|SilentDie _array_element false :!class /Script/FortniteGame.BuildingActor))))))))

(fn OnRep_SpawnDecalTransform
  (Networking|SwitchHasAuthority
    (:Remote
      (Rendering|Decal|SpawnDecalatLocation "/FNGameplayCues/Athena/Prototype/Blueprints/Cube/FX/MD_Decal_Scorch_NoEmissive.MD_Decal_Scorch_NoEmissive" "64.000000,256.000000,256.000000" (.location (Variables|Default|GetSpawnDecalTransform)) (.rotation (Variables|Default|GetSpawnDecalTransform))))))

(fn ConstructionScript)
